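; Blueprint interface for Creative skydive vortex devices, declaring how a vortex reports its push direction for a given actor.
; GetVortexDirection takes a For Actor reference and returns a boolean alongside a direction vector.
; The declared defaults are false and a zero vector, leaving the actual logic to implementers.
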(fn GetVortexDirection
  (params (For%20Actor :type Actor%20Object%20Reference))
  (return false "0, 0, 0"))
