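# Material function that centers a logo texture on the faces of a Creative device volume box and returns it as an inverted scalar mask.
# UV channels 1–3 each serve one axis pair, scaled about center by constants divided by object scale to offset stretching.
# If nodes zero UVs where U exceeds 1, so the three sets sum into one clamped sample.
# A second sample using only the XY set modulates that mask via red and alpha.
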MaterialFunction /Game/Creative/Devices/Generic/MaterialFunctions/MF_CP_Device_VolumeBox_LogoCentering
graph {nodes [
  n1 = Divide(
   A=Constant() {r=1}
   B=MaterialFunctionCall() {materialFunction=&ObjectScale}.Scale Y
  )
  n2 = Divide(
   A=Constant() {desc="0.85 is default, not 1", r=1.25}
   B=MaterialFunctionCall() {materialFunction=&ObjectScale}.Scale Z
  )
  n3 = If(
   A=Constant() {r=1}
   B=ComponentMask(None=TextureCoordinate() {coordinateIndex=2}) {r=true}
   "A > B"=MaterialFunctionCall(UVs=TextureCoordinate() {coordinateIndex=2}, Texture Scale=AppendVector(A=@n1, B=@n2)) {materialFunction=&ScaleUVsByCenter}.UVs
   "A < B"=Constant()
  ) {desc=0}
  n4 = Divide(
   A=Constant() {r=1}
   B=MaterialFunctionCall() {materialFunction=&ObjectScale}.Scale X
  )
  n5 = Divide(
   A=Constant() {desc="0.85 is default, not 1", r=1.25}
   B=MaterialFunctionCall() {materialFunction=&ObjectScale}.Scale Z
  )
  n6 = If(
   A=Constant() {r=1}
   B=ComponentMask(None=TextureCoordinate() {coordinateIndex=1}) {r=true}
   "A > B"=MaterialFunctionCall(UVs=TextureCoordinate() {coordinateIndex=1}, Texture Scale=AppendVector(A=@n4, B=@n5)) {materialFunction=&ScaleUVsByCenter}.UVs
   "A < B"=Constant()
  ) {desc=0}
  n7 = Divide(
   A=Constant() {r=1}
   B=MaterialFunctionCall() {materialFunction=&ObjectScale}.Scale X
  )
  n8 = Divide(
   A=Constant() {desc="0.85 is default, not 1", r=1}
   B=MaterialFunctionCall() {materialFunction=&ObjectScale}.Scale Y
  )
  n9 = If(
   A=Constant() {r=1}
   B=ComponentMask(None=TextureCoordinate() {coordinateIndex=3}) {r=true}
   "A > B"=MaterialFunctionCall(UVs=TextureCoordinate() {coordinateIndex=3}, Texture Scale=AppendVector(A=@n7, B=@n8)) {materialFunction=&ScaleUVsByCenter}.UVs
   "A < B"=Constant()
  ) {desc=0}
  n10 = FunctionInput(Preview=TextureObject() {texture=/Game/Creative/Devices/DeviceZone_Base/Materials/Textures/T_CP_DeviceZone_Frame_01_MutatorZonePlacer_D}) {inputName=VolumeBoxLogo, inputType=FunctionInput_Texture2D}
  n11 = TextureSample(UVs=Add(A=Add(A=@n3, B=@n6), B=@n9), Tex=@n10) {samplerSource=SSM_Clamp_WorldGroupSettings}
  n12 = TextureSample(UVs=@n9, Tex=@n10) {samplerSource=SSM_Clamp_WorldGroupSettings}
  n13 = LinearInterpolate(A=@n12.R, B=OneMinus(None=@n12.A), Alpha=@n12.R)
  FunctionOutput(None=OneMinus(None=Multiply(A=@n11.R, B=@n13)))
 ]}
