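# Unlit translucent material for the placed Skydive device's volume box.
# Panning, object-scaled cloud layers, sampled with explicit derivatives and sphere-masked, combine with a blue vertical gradient; a static switch adds a centered, tinted arrow logo.
# Fresnel and world normal Z masks shape opacity and separate box faces.
# A bounds texture uses UV channels 1 and 2, divided by object scale and re-centered, for a bottom gradient added to fresnel.
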Material /Game/Creative/Devices/Skydive/Materials/M_CP_Device_Skydive_VolumeBox_Placed_01
graph {
 properties {
  bApplyCloudFogging=true
  bScreenSpaceReflections=true
  blendMode=BLEND_Translucent
  pixelDepthOffsetMode=PDOM_Legacy
  shadingModel=MSM_Unlit
  usedShadingModels=Unlit
 }
 nodes [
  n1 = Constant() {r=1}
  n2 = ComponentMask(None=TextureCoordinate() {coordinateIndex=2}) {r=true}
  n3 = TextureCoordinate() {coordinateIndex=2}
  n4 = AppendVector(
   A=Divide(
    A=Constant() {r=1}
    B=MaterialFunctionCall() {materialFunction=&ObjectScale}.Scale Y
   )
   B=Divide(
    A=Constant() {desc="0.85 is default, not 1", r=2}
    B=MaterialFunctionCall() {materialFunction=&ObjectScale}.Scale Z
   )
  )
  n5 = Constant() {r=1}
  n6 = Subtract(A=Add(A=Divide(A=@n3, B=@n4), B=@n5), B=Divide(A=@n5, B=@n4) {desc=Scale the center offset by the scaling factor}) {desc=Re offset to maintain center}
  n7 = Constant()
  n8 = Constant() {r=1}
  n9 = ComponentMask(None=TextureCoordinate() {coordinateIndex=1}) {r=true}
  n10 = TextureCoordinate() {coordinateIndex=1}
  n11 = AppendVector(
   A=Divide(
    A=Constant() {r=1}
    B=MaterialFunctionCall() {materialFunction=&ObjectScale}.Scale X
   )
   B=Divide(
    A=Constant() {desc="0.85 is default, not 1", r=2}
    B=MaterialFunctionCall() {materialFunction=&ObjectScale}.Scale Z
   )
  )
  n12 = Constant() {r=1}
  n13 = Subtract(A=Add(A=Divide(A=@n10, B=@n11), B=@n12), B=Divide(A=@n12, B=@n11) {desc=Scale the center offset by the scaling factor}) {desc=Re offset to maintain center}
  n14 = Constant()
  n15 = TextureSample(UVs=Add(A=If(A=@n1, B=@n2, "A > B"=@n6, "A < B"=@n7) {desc=0}, B=If(A=@n8, B=@n9, "A > B"=@n13, "A < B"=@n14) {desc=0})) {
   samplerSource=SSM_Clamp_WorldGroupSettings
   texture=/Game/Creative/Devices/DeviceZone_Base/Materials/Textures/T_CP_DeviceZone_Bounds_01_BarrierPlacer_D
  }
  n16 = MaterialFunctionCall(
   Brightness=ScalarParameter() {defaultValue=0.5, parameterName=FresnelBrightness}
   Exponent=ScalarParameter() {defaultValue=0.25, parameterName=FresnelExponent}
  ) {materialFunction=&MF_Fresnel_FX}
  Add(
   A=@n16.Fresnel Inverted Double Sided
   B=Multiply(A=@n15.G, B=ScalarParameter() {defaultValue=0.3, parameterName=BottomGradientOpacity}) {constB=0.5}
  )
  n17 = OneMinus(None=SphereMask(
    A=TextureCoordinate()
    B=Constant() {r=0.5}
    Radius=ScalarParameter() {defaultValue=0.85, group=Clouds, parameterName=Clouds Sphere Mask Radius}
    Hardness=ScalarParameter() {group=Clouds, parameterName=Clouds Sphere Mask Hardness}
   ))
  n18 = Multiply(A=ComponentMask(None=VertexNormalWS()) {b=true}) {constB=-1}
  n19 = Multiply(A=OneMinus(None=Saturate(None=@n18)), B=@n16.Fresnel Inverted Double Sided)
  Multiply(A=LinearInterpolate(A=@n19, B=@n17), B=ScalarParameter() {defaultValue=1, parameterName=Opacity_Value})
  n20 = OneMinus(None=ComponentMask(None=VertexNormalWS()) {b=true})
  Reroute(None=Multiply(A=@n19, B=Multiply(A=@n20, B=@n16.Fresnel Inverted Double Sided)))
  n21 = ComponentMask(None=VertexNormalWS()) {b=true}
  n22 = OneMinus(None=Saturate(None=Multiply(A=@n21)))
  n23 = MaterialFunctionCall!(VolumeBoxLogo=TextureObjectParameter() {parameterName=VolumeBoxLogoTexture, texture=/Game/Creative/Devices/Skydive/Materials/Textures/T_CP_Device_Skydive_VolumeArrow_01_D}) {materialFunction=/Game/Creative/Devices/Generic/MaterialFunctions/MF_CP_Device_VolumeBox_LogoCentering}
  n24 = OneMinus(None=LinearInterpolate(
    A=VectorParameter() {group=Logo, parameterName=Logo Color, sortPriority=1, defaultValue=<rgba 0.8148 0.8148 0.8148 1>}.RGB
    B=OneMinus(None=@n23.Result)
   ))
  n25 = Constant() {r=5}
  n26 = Multiply(
   A=@n22
   B=Multiply(A=MaterialFunctionCall(In=@n24, Contrast=@n25) {materialFunction=&CheapContrast}.Result)
  )
  n27 = VectorParameter() {group=BG, parameterName=BG Color, sortPriority=1, defaultValue=<rgba 0 0.2173 0.8255 1>}
  n28 = VectorParameter() {group=BG, parameterName=BG Bottom Glow Color, sortPriority=2, defaultValue=<rgba 0.0609 0.6532 1 1>}
  n29 = Clamp(None=MaterialFunctionCall() {materialFunction=&LinearGradient}.VGradient)
  n30 = AppendVector(
   A=ScalarParameter() {defaultValue=0.15, group=Clouds, parameterName=Clouds PanSpeed U}
   B=ScalarParameter() {defaultValue=1, group=Clouds, parameterName=Clouds PanSpeed V}
  )
  n31 = TextureCoordinate()
  n32 = ScalarParameter() {defaultValue=1, group=Clouds, parameterName=Clouds UV TextureCoordinates}
  n33 = MaterialFunctionCall() {materialFunction=&ObjectScale}
  n34 = Multiply(A=Constant() {r=1}, B=@n33.Scale X)
  n35 = Multiply(A=@n33.Scale Y, B=Divide(A=@n33.Scale Z, B=@n33.Scale Z))
  n36 = Multiply(A=Multiply(A=Multiply(A=@n31, B=@n32), B=AppendVector(A=@n34, B=@n35)), B=Constant() {r=1})
  n37 = Panner(Coordinate=@n36, Speed=@n30) {speedY=0.002}
  n38 = TextureObjectParameter() {group=Clouds, parameterName=Clouds Texture, sortPriority=1, texture=/Game/Effects/Fort_Effects/Textures/Tile/T_CloudScape_08}
  n39 = Divide(A=@n36) {constB=6}
  n40 = DDX(Value=@n39)
  n41 = DDY(Value=@n39)
  n42 = TextureSample!(UVs=Panner(Coordinate=@n36, Speed=@n30) {speedX=-0.0001, speedY=0.0025}, Tex=@n38, "DDX(UVs)"=@n40, "DDY(UVs)"=@n41, Apply View MipBias) {mipValueMode=TMVM_Derivative}
  n43 = Desaturation(None=LinearInterpolate(A=TextureSample(UVs=@n37, Tex=@n38, "DDX(UVs)"=@n40, "DDY(UVs)"=@n41) {mipValueMode=TMVM_Derivative}.RGB, B=@n42.RGB))
  n44 = Multiply(A=@n42.G, B=@n17)
  n45 = Multiply(
   A=Power(Base=Multiply(A=@n43, B=@n44), Exp=ScalarParameter() {defaultValue=2, group=Clouds, parameterName=Clouds Mask Power})
   B=ScalarParameter() {defaultValue=4, group=Clouds, parameterName=Clouds Mask Multiplier}
  )
  n46 = LinearInterpolate(
   A=Multiply(A=Clamp(None=@n45), B=ScalarParameter() {defaultValue=1, group=Clouds, parameterName=Clouds Alpha})
   B=LinearInterpolate(A=@n27.RGB, B=LinearInterpolate(A=@n27.RGB, B=@n28.RGB, Alpha=@n29), Alpha=@n22)
  )
  Reroute(None=StaticSwitchParameter(True=LinearInterpolate(A=@n26, B=@n46), False=@n46) {defaultValue=true, desc=Mapped Using UV Channel 1, group=Logo, parameterName=Use Logo, sortPriority=0})
 ]
}
